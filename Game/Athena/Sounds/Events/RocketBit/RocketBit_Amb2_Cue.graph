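# RocketBit_Amb2_Cue is a looping, spatialized ambient sound for the RocketBit event, routed to the Athena ambient sound class.
# A mixer layers a half-volume Grumpy beam loop with two Ostrich target-laser loops, leaving a fourth input empty.
# Modulators pin the beam at normal pitch and drop one laser loop to half pitch.
# Overridden attenuation uses short-range falloff and listener focus to quieten it outside the focus cone.
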SoundCue /Game/Athena/Sounds/Events/RocketBit/RocketBit_Amb2_Cue
root=Mixer_0
nodes [
 {id=Mixer_0, children [Modulator_0,Modulator_1,WavePlayer_3,null], properties {inputVolume [0.5,1,1,1]}}
 {id=Modulator_0, children [WavePlayer_0], properties {pitchMax=1, pitchMin=1}}
 {id=Modulator_1, children [WavePlayer_2], properties {pitchMax=0.5, pitchMin=0.5}}
 {id=WavePlayer_0, properties {bLooping=true, soundWaveAssetPtr=/Game/Athena/Sounds/Events/Dopey/Grumpy/Grumpy_Beam_LoopA_02}}
 {id=WavePlayer_2, properties {bLooping=true, soundWaveAssetPtr=/Game/Athena/Sounds/Vehicles/Ostrich/Veh_Ostrich_TargetLaser_Loop_03}}
 {id=WavePlayer_3, properties {bLooping=true, soundWaveAssetPtr=/Game/Athena/Sounds/Vehicles/Ostrich/Veh_Ostrich_TargetLaser_Loop_01}}
]
properties {
 bOverrideAttenuation=true
 maxDistance=1024
 soundClassObject=&Fort_Ambient_Athena
 volumeMultiplier=0.624
 attenuationOverrides {
  bEnableFocusInterpolation=true
  bEnableListenerFocus=true
  falloffDistance=512
  focusAzimuth=15
  nonFocusAzimuth=30
  nonFocusVolumeAttenuation=0.5
  stereoSpread=128
  attenuationShapeExtents {x=512}
 }
}
